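# MF_SortTranslucency is a library material function that masks a translucent Opacity input by depth, apparently for the sorted outline material.
# It derives pixel view-space depth from world position relative to the camera.
# Saturated differences from CustomDepth and SceneDepth multiply into Opacity, fading out pixels behind either depth.
# An If-based hard-cutoff variant and two PixelDepth nodes are left unconnected.
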MaterialFunction /Game/Global/Materials/MaterialGym/Materials/SortedOutline/MF_SortTranslucency
graph {
 properties {bExposeToLibrary=true, previewBlendMode=BLEND_Translucent, libraryCategoriesText [Misc, Tech Art]}
 nodes [
  n1 = FunctionInput() {inputName=Opacity, inputType=FunctionInput_Scalar}
  n2 = ComponentMask(None=SceneTexture() {sceneTextureId=PPI_CustomDepth}.Color) {r=true}
  n3 = Subtract(A=WorldPosition().XYZ, B=CameraPositionWS())
  n4 = Multiply(A=ComponentMask(None=Transform(None=@n3) {transformSourceType=TRANSFORMSOURCE_World, transformType=TRANSFORM_View}) {b=true}) {constB=1}
  n5 = Saturate(None=Subtract(A=Add(A=@n2), B=@n4))
  n6 = SceneTexture() {sceneTextureId=PPI_SceneDepth}
  n7 = Saturate(None=Subtract(A=ComponentMask(None=@n6.Color) {r=true}, B=@n4))
  FunctionOutput(None=Multiply(A=@n1, B=Multiply(A=@n5, B=@n7)))
  n8 = Transform(None=Subtract(A=WorldPosition().XYZ, B=CameraPositionWS())) {transformSourceType=TRANSFORMSOURCE_World, transformType=TRANSFORM_View}
  If(
   A=Multiply(A=ComponentMask(None=@n8) {b=true}) {constB=1}
   B=ComponentMask(None=SceneTexture() {sceneTextureId=PPI_CustomDepth}.Color) {r=true}
   "A > B"=Constant()
   "A < B"=@n1
  )
  PixelDepth()
  PixelDepth()
 ]
}
