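# Unlit additive material in the MaterialGym folder, named for a sorted outline effect on instanced, skeletal or deformed meshes.
# It allows translucent custom depth writes and feeds a near-zero Opacity parameter into the shared MF_SortTranslucency function.
# A separate branch multiplies the EmissiveColor parameter by itself, with no output connection shown.
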Material /Game/Global/Materials/MaterialGym/Materials/SortedOutline/M_SortedOutline
graph {
 properties {
  allowTranslucentCustomDepthWrites=true
  bAutomaticallySetUsageInEditor=false
  bCastRayTracedShadows=false
  bUsedWithInstancedStaticMeshes=true
  bUsedWithMeshDeformer=true
  bUsedWithSkeletalMesh=true
  blendMode=BLEND_Additive
  opacityMaskClipValue=0.001
  pixelDepthOffsetMode=PDOM_Legacy
  shadingModel=MSM_Unlit
  translucencyLightingMode=TLM_Surface
  usedShadingModels=Unlit
 }
 nodes [
  MaterialFunctionCall!(Opacity=ScalarParameter() {defaultValue=0.001, parameterName=Opacity, sliderMax=1, sliderMin=0.001}) {materialFunction=&/MF_SortTranslucency}
  n1 = VectorParameter() {parameterName=EmissiveColor, defaultValue=<rgba 0.5 0 1 1>}
  Multiply(A=@n1.RGB, B=@n1.RGB)
 ]
}
