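# Translucent unlit overlay material that draws an inverted-hull outline on static instanced or skeletal meshes.
# A LineThickness parameter scales the world-space vertex normal to push the hull outward, and a squared EmissiveColor provides the line color.
# Opacity passes through MF_SortTranslucency and is multiplied by one minus TwoSidedSign, so only back faces show.
# Translucent custom depth writes and velocity output are enabled.
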Material /Game/Global/Materials/MaterialGym/Materials/SortedOutline/M_SortedOutline_Overlay
graph {
 properties {
  allowTranslucentCustomDepthWrites=true
  bAutomaticallySetUsageInEditor=false
  bCastRayTracedShadows=false
  bOutputTranslucentVelocity=true
  bUsedWithInstancedStaticMeshes=true
  bUsedWithMeshDeformer=true
  bUsedWithSkeletalMesh=true
  blendMode=BLEND_Translucent
  opacityMaskClipValue=0.001
  pixelDepthOffsetMode=PDOM_Legacy
  shadingModel=MSM_Unlit
  translucencyLightingMode=TLM_Surface
  twoSided=true
  usedShadingModels=Unlit
 }
 nodes [
  Multiply(
   A=MaterialFunctionCall!(Opacity=ScalarParameter() {defaultValue=1, parameterName=Opacity, sliderMax=1, sliderMin=0.001}) {materialFunction=&/MF_SortTranslucency}.Result
   B=OneMinus(None=TwoSidedSign())
  )
  n1 = VectorParameter() {parameterName=EmissiveColor, defaultValue=<rgba 0.5 0 1 1>}
  Multiply(A=@n1.RGB, B=@n1.RGB)
  Multiply(
   A=ScalarParameter() {defaultValue=0.3, group=01_InvertedHull_Outlines, parameterName=LineThickness, sliderMax=0.5, sortPriority=3}
   B=VertexNormalWS()
  )
 ]
}
